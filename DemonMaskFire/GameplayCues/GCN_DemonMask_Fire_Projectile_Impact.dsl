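; Burst gameplay cue for the Demon Mask fire projectile impact, playing effects and a local sound indicator.
; After the parent call, scalable-float flags pick one of three Niagara impact versions, aligned to the hit normal.
; It spawns a radial force field actor only off mobile, outside PC performance mode and at higher detail.
; A raw-magnitude switch, marked as a test hack, spawns extra impact variants for comparison on cooked builds.
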(fn OnBurst(EitherVFX-preferthisversion)
  (params (MyTarget :type Actor%20Object%20Reference) (Parameters :type Gameplay%20Cue%20Parameters%20Structure%20%28by%20ref%29) (ParticleComponents :type Array%20of%20FXSystem%20Component%20Object%20References) (AudioComponents :type Array%20of%20Audio%20Component%20Object%20References) (BurstCameraShakeInstance :type Camera%20Shake%20Base%20Object%20Reference) (BurstDecalInstance :type Decal%20Actor%20Object%20Reference))
  (bind (_normalizedmagnitude _rawmagnitude) (Ability|GameplayCue|BreakGameplayCueParameters Parameters))
  (bind (_normalizedmagnitude_2 _rawmagnitude_2 _effectcontext _matchedtagname _originaltag _aggregatedsourcetags _aggregatedtargettags _location _normal) (Ability|GameplayCue|BreakGameplayCueParameters Parameters))
  (bind _returnvalue (Math|Conversions|RotationFromXVector _normal))
  (bind _returnvalue_2 (Transformation|GetActorLocation MyTarget))
  (|CallParentFunction MyTarget Parameters ParticleComponents AudioComponents BurstCameraShakeInstance BurstDecalInstance)
  (Utilities|IsValid MyTarget
    (:Is%20Valid
      (Fortnite|ShowSoundIndicatorLocal :WorldContextObject MyTarget :Location _returnvalue_2 :MaxAudibleDistance 0.0 :IndicatorType "Gunshot" :Tint "(R=0.000000,G=0.000000,B=0.000000,A=0.000000)" :SoundBase "/Game/Sounds/Fort_Weapons/Launchers/ElephantGarlic/ElephantGarlic_Explode_Athena_Cue.ElephantGarlic_Explode_Athena_Cue" :AttenuationMaxRangeMultiplier 1.0 :AttenuationIntensityMultiplier 1.0)
      (if (ScalableFloat|GetValueasBool (Variables|Default|GetHF_UseBPGraphVFX))
        (if (ScalableFloat|GetValueasBool (Variables|Default|GetHF_DisableOldImpactVFX))
          (bind _returnvalue_3 (Effects|Components|ParticleSystem|SpawnFXAtLocation 0 "/DemonMaskFire/Assets/VFX/NS_DemonMask_Fire_Prj_Impact.NS_DemonMask_Fire_Prj_Impact" _location _returnvalue "1.000000,1.000000,1.000000" true "AutoRelease"))
          (else
            (bind _returnvalue_4 (Effects|Components|ParticleSystem|SpawnFXAtLocation 0 "/DemonMaskFire/Assets/VFX/NS_DemonMask_Fire_Prj_Impact_OLD.NS_DemonMask_Fire_Prj_Impact_OLD" _location _returnvalue "1.000000,1.000000,1.000000" true "AutoRelease"))))
        (else
          (bind _returnvalue_5 (Effects|Components|ParticleSystem|SpawnFXAtLocation 0 "/DemonMaskFire/Assets/VFX/NS_DemonMask_Fire_Prj_Impact_v02.NS_DemonMask_Fire_Prj_Impact_v02" _location _returnvalue "1.000000,1.000000,1.000000" true "AutoRelease"))))
      ;[0.4769 0.8897 1] RADIAL FORCE
      (Effects|Components|ParticleSystem|SetIntParameter _returnvalue_5 "Color Variant" 0)
      (if (Math|Boolean|ANDBoolean (not (Fortnite|IsMobilePlatform)) (not (Class|PlatformHelpersLibrary|IsPcPerfMode 0)) (> (Rendering|GetRenderingDetailMode) 1))
        (Game|SpawnActorfromClass 0 "/Game/Effects/Niagara/Blueprints/PhsyicsFields/FN_RadialForce.FN_RadialForce_C" (Math|Transform|MakeTransform _returnvalue_2) "Undefined" "MultiplyWithRoot" 0 (Variables|Default|GetPlayrate) (Variables|Default|GetTime%20Decay) (Variables|Default|GetImpact%20Radius) (Variables|Default|GetStrength) (Variables|Default|GetDelay%20Time) 0))
      (if (> _rawmagnitude 0)
        ;[0.9479 0 0.07653 0.5] [CHEAT HACK] - for testing multiple versions of the impact FX on a cooked build
        (switch int (Math|Float|Truncate _rawmagnitude)
          (:Default
            _)
          (:0
            _)
          (:1
            (Effects|Components|ParticleSystem|SpawnFXAtLocation :FXSystemAsset "/DemonMaskFire/Assets/VFX/NS_DemonMask_Fire_Prj_Impact.NS_DemonMask_Fire_Prj_Impact" :Location _returnvalue_2 :Scale "1.000000,1.000000,1.000000" :bAutoDestroy true :PoolingMethod "AutoRelease" :bAutoActivate true))
          (:2
            (Effects|Components|ParticleSystem|SpawnFXAtLocation :FXSystemAsset "/DemonMaskFire/Assets/VFX/NS_DemonMask_Fire_Prj_Impact_OLD.NS_DemonMask_Fire_Prj_Impact_OLD" :Location _returnvalue_2 :Scale "1.000000,1.000000,1.000000" :bAutoDestroy true :PoolingMethod "AutoRelease" :bAutoActivate true))
          (:3
            (Effects|Components|ParticleSystem|SpawnFXAtLocation :FXSystemAsset "/DemonMaskFire/Assets/VFX/NS_DemonMask_Fire_Prj_Impact_v02.NS_DemonMask_Fire_Prj_Impact_v02" :Location _returnvalue_2 :Scale "1.000000,1.000000,1.000000" :bAutoDestroy true :PoolingMethod "AutoRelease" :bAutoActivate true))
          (:4
            (Effects|Components|ParticleSystem|SpawnFXAtLocation :FXSystemAsset "/DemonMaskFire/Assets/VFX/NS_DemonMask_Fire_Prj_Impact_v02.NS_DemonMask_Fire_Prj_Impact_v02" :Location _returnvalue_2 :Scale "1.000000,1.000000,1.000000" :bAutoDestroy true :PoolingMethod "AutoRelease" :bAutoActivate true))
          (:5
            (Effects|Components|ParticleSystem|SpawnFXAtLocation :FXSystemAsset "/DemonMaskFire/Assets/VFX/NS_DemonMask_Fire_Prj_Impact_v02.NS_DemonMask_Fire_Prj_Impact_v02" :Location _returnvalue_2 :Scale "1.000000,1.000000,1.000000" :bAutoDestroy true :PoolingMethod "AutoRelease" :bAutoActivate true))
          (:6
            (Effects|Components|ParticleSystem|SpawnFXAtLocation :FXSystemAsset "/DemonMaskFire/Assets/VFX/NS_DemonMask_Fire_Prj_Impact_v02.NS_DemonMask_Fire_Prj_Impact_v02" :Location _returnvalue_2 :Scale "1.000000,1.000000,1.000000" :bAutoDestroy true :PoolingMethod "AutoRelease" :bAutoActivate true))))
      (return))))
